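; Controller component that manages the Beast Girl transformation, tracking the pawn swap and the durability of the item involved.
; On possession change it caches the first player pawn; swapping to a stand-in pawn arms a one-shot controller-changed handler on the old pawn.
; When that pawn regains a controller, it plays a normal or weakened end montage, chosen by bOptingOut.
; Helpers derive clamped durability from final Beast Girl health and check durability against the weapon stat row's per-use cost.
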(event EventBeginPlay
  (|CallParentFunction)
  (bind _ascontroller (Utilities|Casting|CastToController (Components|GetOwner)))
  (Pawn|AssignOnPossessedPawnChanged _ascontroller (delegate Custom|OnPossessedPawnChanged_Event)))

(event Custom|OnPossessedPawnChanged_Event
  (params (OldPawn :type Pawn%20Object%20Reference) (NewPawn :type Pawn%20Object%20Reference))
  (Utilities|FlowControl|Sequence
    (:then_0
      (Utilities|Casting|CastToBeastGirlStandInPawn NewPawn)
      (bind _asfort_player_pawn (Utilities|Casting|CastToFortPlayerPawn OldPawn))
      (Pawn|AssignReceiveControllerChangedDelegate _asfort_player_pawn (EventDispatchers|CreateEvent :!event Custom|HandlePossessed)))
    (:then_1
      (Utilities|IsValid (Variables|Default|GetCachedOriginalPawn)
        (:Is%20Not%20Valid
          (bind _asfort_player_pawn_2 (Utilities|Casting|CastToFortPlayerPawn NewPawn))
          (Variables|Default|SetCachedOriginalPawn _asfort_player_pawn_2))))))

(event Custom|HandlePossessed
  (params (Pawn :type Pawn%20Object%20Reference) (OldController :type Controller%20Object%20Reference) (NewController :type Controller%20Object%20Reference))
  (Utilities|IsValid NewController
    (:Is%20Valid
      (bind _ascharacter (Utilities|Casting|CastToCharacter Pawn))
      (Animation|Montage|PlayMontage (Variables|/Script/Engine.Character|GetMesh _ascharacter) (select (Variables|BeastGirlTransformationManager|GetbOptingOut) (Variables|Default|GetEndMontage) (Variables|Default|GetEndMontageWeakened)))
      (Pawn|UnbindEventfromReceiveControllerChangedDelegate _ascharacter (EventDispatchers|CreateEvent :!event Custom|HandlePossessed)))))

(fn GetNewDurability
  (params (FinalBeastGirlHealth :type Float%20%28single-precision%29) (ItemDefinition :type Fort%20World%20Item%20Definition%20Object%20Reference))
  (bind _asfort_weapon_item_definition (Utilities|Casting|CastToFortWeaponItemDefinition ItemDefinition)
    (:then
      (bind _returnvalue (Utilities|GetDataTableRow (Utilities|Struct|BreakDataTableRowHandle (WeaponItemDefinition|GetWeaponStatHandle _asfort_weapon_item_definition)) (.RowName (Utilities|Struct|BreakDataTableRowHandle (WeaponItemDefinition|GetWeaponStatHandle _asfort_weapon_item_definition))))
        (:then
          (return (Math|Float|Clamp(Float) (- FinalBeastGirlHealth (ScalableFloat|GetValueatLevel (Variables|Default|GetHF_DurabilityDecreasePerTransform))) (Math|Float|Min(Float) (ScalableFloat|GetValueatLevel (Variables|Default|GetHF_MinDurability)) (.DurabilityPerUse (Utilities|Struct|BreakFortRangedWeaponStats _returnvalue))) (Item|GetMaxDurability _asfort_weapon_item_definition 0 :!class /Script/FortniteGame.FortWorldItemDefinition))))
        (:RowNotFound
          (return 0.0))))
    (:CastFailed
      (return 0.0))))

(fn HasEnoughDurability
  (params (CurrentWeapon :type Fort%20Weapon%20Object%20Reference))
  (bind (_datatable _rowname) (Utilities|Struct|BreakDataTableRowHandle (Data|GetWeaponDataWeaponStatHandle CurrentWeapon)))
  (bind (_bisvalid _returnvalue) (FortWeapon|GetDurability CurrentWeapon))
  (bind _returnvalue_2 (Utilities|GetDataTableRow _datatable _rowname)
    (:then
      (return (> _returnvalue (.DurabilityPerUse (Utilities|Struct|BreakFortRangedWeaponStats _returnvalue_2)))))
    (:RowNotFound
      (return false))))
